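# Unlit, masked wireframe material from the environment art-tool overlay set, apparently meant to draw mesh edges over geometry.
# It exposes a cyan WireframeColor, an Opacity scalar and a WireframeOffset that feeds the engine CameraOffset world-position-offset function.
# A separate branch scales the vertex normal by a constant, but links to material outputs are not shown.
# Ray-traced shadow casting and automatic usage flagging in the editor are disabled.
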Material /Game/ArtTools/Environment/OverlayMaterials/M_WireframeOverlay
graph {
 properties {
  bAutomaticallySetUsageInEditor=false
  bCastRayTracedShadows=false
  blendMode=BLEND_Masked
  shadingModel=MSM_Unlit
  usedShadingModels=Unlit
  wireframe=true
 }
 nodes [
  MaterialFunctionCall(Offset Amount=ScalarParameter() {defaultValue=3, parameterName=WireframeOffset}) {materialFunction=/Engine/Functions/Engine_MaterialFunctions02/WorldPositionOffset/CameraOffset}
  Multiply(A=VertexNormalWS()) {constB=3}
  ScalarParameter() {defaultValue=1, parameterName=Opacity}
  VectorParameter() {parameterName=WireframeColor, defaultValue=<rgba 0 0.7305 1 1>}
 ]
}
